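; Blueprint Interface for team member info entries in the battle royale HUD squad list.
; It declares a single function, SetTeamMemberPlayerState, which takes a Player State object reference, letting callers assign a player to an implementing object without knowing its concrete class.
; The file holds only the signature, with no implementation.
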(fn SetTeamMemberPlayerState
  (params (PlayerState :type Player%20State%20Object%20Reference)))
