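# Squad list HUD row for a player's name on AthenaPlayerNameInfoBase, with marker, name, platform icon, voice indicator and ranked icon.
# PreConstruct picks the PC or simpler mobile layout in a widget switcher from bShouldShowMobileLayout.
# UpdateRankedProgress looks up tier data through Habanero display data and overrides, showing the small simple icon when the lookup succeeds.
# The rank icon exists only in the PC layout and collapses on reset or before each update.
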WidgetBlueprint /BattleRoyaleGameUI/HUD/SquadList/WBP_BRGameUI_PlayerNameInfo
graphs [EventGraph {nodes [
   n1 = Event UpdateRankedProgress!() -> (OutputDelegate, then, RankedProgress:"Ranked Progress Replicated Data Structure (by ref)") {is=title K2Node_Event}
   > SetVisibility(self=Get LazyImage_RankedIcon() -> (LazyImage_RankedIcon:Fort Lazy Image Object Reference), InVisibility=Collapsed)
   n2 = Break Ranked Progress Replicated Data!(RankedProgressReplicatedData:"Ranked Progress Replicated Data Structure (by ref)"=@n1.RankedProgress) -> (RankType:String, Rank:int, HighestRank:int, CurrentMMR:int, CurrentPosition:int) {is=title K2Node_BreakStruct}
   n3 = GetHabaneroDisplayDataOverride!(
    InDisplayData:Fort Habanero Display Data Object Reference=Get Rank Data() -> (Rank Data:Fort Habanero Display Data Object Reference)
    InOverrideData:Fort Compete UIHabanero Tier Override List Object Reference=Get RankInfoOverrides() -> (RankInfoOverrides:Fort Compete UIHabanero Tier Override List Object Reference)
    InKey:"Fort Compete UIHabanero Tier Override Key Structure (by ref)"=Make Fort Compete UIHabanero Tier Override Key!(
     RankedTrack:Name=String To Name(InString=@n2)
     Season:String
     Rank:int=@n2.Rank
     MatchmakingRating:int=@n2.CurrentMMR
     CurrentPosition:int=@n2.CurrentPosition
    ) -> (FortCompeteUIHabaneroTierOverrideKey:Fort Compete UIHabanero Tier Override Key Structure) {is=title K2Node_MakeStruct}
   ) -> (OutData:Fort Habanero Tier Override Structure, ReturnValue:bool)
   > Branch(Condition=@n3.ReturnValue)
   n4 = Get LazyImage_RankedIcon()
   > SetBrushFromSoftTexture!(
    execute
    self:Image Object Reference=@n4
    SoftTexture:Texture 2D Soft Object Reference=Break Fort Habanero Tier!(FortHabaneroTier:"Fort Habanero Tier Structure (by ref)"=@n3) -> (
     ForegroundIcon:Material Instance Soft Object Reference
     ForegroundIconSmall:Material Instance Soft Object Reference
     BackgroundIcon:Material Instance Soft Object Reference
     BackgroundIconSmall:Material Instance Soft Object Reference
     WhiteIcon:Texture 2D Soft Object Reference
     SimpleIconSmall:Texture 2D Soft Object Reference
     Name:Text
     TierType:Fort Habanero Tier Type Data Object Reference
     AdditionalLeftIcon:Material Instance Soft Object Reference
     AdditionalRightIcon:Material Instance Soft Object Reference
    ) {is=title K2Node_BreakStruct}.SimpleIconSmall
    bMatchSize:bool=false
   ) -> (then)
   > SetVisibility(self=@n4, InVisibility=HitTestInvisible)
   n5 = Event PreConstruct()
   > Parent: PreConstruct(IsDesignTime=@n5.IsDesignTime)
   > SetActiveWidget(
    self:Widget Switcher Object Reference=Get WidgetSwitcher() -> (WidgetSwitcher:Common Activatable Widget Switcher Object Reference)
    Widget=Select(
     Option 0:Widget Object Reference=Get PC_PlayerNameInfo() -> (PC_PlayerNameInfo:Horizontal Box Object Reference)
     Option 1:Widget Object Reference=Get Mobile_PlayerNameInfo() -> (Mobile_PlayerNameInfo:Horizontal Box Object Reference)
     Index=Get bShouldShowMobileLayout() -> (bShouldShowMobileLayout:bool)
    ) -> (ReturnValue:Widget Object Reference)
   )
   Event ResetRankedProgress() {is=title K2Node_Event}
   > SetVisibility(self=Get LazyImage_RankedIcon(), InVisibility=Collapsed)
  ]}]
variables [
 bShouldShowMobileLayout {type=boolean}
 Rank Data {type=FortHabaneroDisplayData, default=/Game/UI/Habanero/HabaneroDisplayData}
 RankInfoOverrides {type=FortCompeteUIHabaneroTierOverrideList, default=/Game/UI/Habanero/DA_RankDisplayDataOverride}
]
widgets {
 info {
  inheritedWidgetCount=7
  parentClass=/Script/FortniteUI.AthenaPlayerNameInfoBase
  rootWidgetClass=&CommonActivatableWidgetSwitcher
  widgetCount=17
 }
 widgets [
  WidgetSwitcher {widgetClassPath=&CommonActivatableWidgetSwitcher, bIsVariable=true, widget=WidgetSwitcher}
  SizeBox_1 {
   widgetClassPath=&SizeBox
   parent=WidgetSwitcher
   widget=SizeBox_1
   properties {bOverride_HeightOverride=true, heightOverride=24}
   slot {class=WidgetSwitcherSlot}
  }
  PC_PlayerNameInfo {
   widgetClassPath=&HorizontalBox
   bIsVariable=true
   parent=SizeBox_1
   widget=PC_PlayerNameInfo
   properties {flowDirectionPreference=Culture}
   slot {class=SizeBoxSlot, properties {padding=<ltrb 0 0 5 0>}}
  }
  LazyImage_RankedIcon {
   widgetClassPath=&FortLazyImage
   bIsVariable=true
   parent=PC_PlayerNameInfo
   widget=LazyImage_RankedIcon
   properties {
    visibility=Collapsed
    brush {
     bIsDynamicallyLoaded=false
     drawAs=Image
     imageType=NoImage
     mirroring=NoMirror
     resourceName=None
     resourceObject=null
     tiling=NoTile
     imageSize=<24 24>
     margin=<ltrb 0 0 0 0>
     outlineSettings {
      bUseBrushTransparency=false
      roundingType=HalfHeightRadius
      width=0
      color {colorUseRule=UseColor_Specified, specifiedColor=<rgba 0 0 0 0>}
      cornerRadii=<0 0 0 0>
     }
     tintColor {colorUseRule=UseColor_Specified, specifiedColor=<rgba 1 1 1 1>}
     uVRegion {bIsValid=false, max=<0 0>, min=<0 0>}
    }
    renderTransform {angle=0, scale=<1.4 1.4>, shear=<0 0>, translation=<0 0>}
   }
   slot {class=HorizontalBoxSlot, properties {horizontalAlignment=HAlign_Center, verticalAlignment=VAlign_Center, padding=<ltrb 0 0 4 0>}}
  }
  Image_Marker {
   widgetClassPath=&Image
   bInherited=true
   bIsVariable=true
   parent=PC_PlayerNameInfo
   widget=Image_Marker
   properties {
    visibility=SelfHitTestInvisible
    brush {
     bIsDynamicallyLoaded=false
     drawAs=Image
     imageType=FullColor
     mirroring=NoMirror
     resourceName=None
     resourceObject=/BattleRoyaleGameUI/HUD/SquadList/Assets/M_UI_AngledBulletPoint
     tiling=NoTile
     imageSize=<12 12>
     margin=<ltrb 0 0 0 0>
     outlineSettings {
      bUseBrushTransparency=false
      roundingType=HalfHeightRadius
      width=0
      color {colorUseRule=UseColor_Specified, specifiedColor=<rgba 0 0 0 0>}
      cornerRadii=<0 0 0 0>
     }
     tintColor {colorUseRule=UseColor_Specified, specifiedColor=<rgba 1 1 1 1>}
     uVRegion {bIsValid=false, max=<0 0>, min=<0 0>}
    }
    renderTransform {angle=0, scale=<1 1.25>, shear=<0 0>, translation=<0 -1>}
   }
   slot {class=HorizontalBoxSlot, properties {horizontalAlignment=HAlign_Left, verticalAlignment=VAlign_Center, padding=<ltrb 0 0 2 0>}}
  }
  OverlayNameandLevel {
   widgetClassPath=&Overlay
   parent=PC_PlayerNameInfo
   widget=OverlayNameandLevel
   slot {class=HorizontalBoxSlot, properties {verticalAlignment=VAlign_Center}}
  }
  CommonTextBlock_PlayerName {
   widgetClassPath=&CommonTextBlock
   bInherited=true
   bIsVariable=true
   parent=OverlayNameandLevel
   widget=CommonTextBlock_PlayerName
   properties {
    text=TestPlayerName
    font {
     bForceMonospaced=false
     bMaterialIsStencil=false
     fontMaterial=/BattleRoyaleGameUI/Shared/M_UI_Font_White
     fontObject=&BurbankSmall
     letterSpacing=0
     monospacedWidth=1
     size=12
     skewAmount=0
     typefaceFontName=Black
     outlineSettings {
      bApplyOutlineToDropShadows=false
      bMiteredCorners=false
      bSeparateFillAlpha=true
      outlineMaterial=/BattleRoyaleGameUI/Shared/MI_UI_Font_SoftShadow_SmallText
      outlineSize=4
      outlineColor=<rgba 1 1 1 1>
     }
    }
    margin=<ltrb -2 -2 -2 -2>
    shadowOffset=<0 0>
   }
   slot {class=OverlaySlot, properties {horizontalAlignment=HAlign_Right, verticalAlignment=VAlign_Center}}
  }
  Spacer_111 {widgetClassPath=&Spacer, parent=PC_PlayerNameInfo, widget=Spacer_111, properties {size=<5 1>}, slot {class=HorizontalBoxSlot}}
  SizeBox_0 {
   widgetClassPath=&SizeBox
   parent=PC_PlayerNameInfo
   widget=SizeBox_0
   properties {
    bOverride_MaxDesiredHeight=true
    bOverride_MaxDesiredWidth=true
    heightOverride=22
    maxDesiredHeight=20
    maxDesiredWidth=20
    widthOverride=22
   }
   slot {class=HorizontalBoxSlot, properties {horizontalAlignment=HAlign_Center, verticalAlignment=VAlign_Center, padding=<ltrb 0 -5 0 -5>}}
  }
  Image_PlatformIcon {
   widgetClassPath=&Image
   bInherited=true
   bIsVariable=true
   parent=SizeBox_0
   widget=Image_PlatformIcon
   properties {
    renderOpacity=0.9
    visibility=Collapsed
    brush {
     bIsDynamicallyLoaded=false
     drawAs=Image
     imageType=NoImage
     mirroring=NoMirror
     resourceName=None
     resourceObject=/Game/UI/Friends_UI/Social/Console_PlatformIcon_32x
     tiling=NoTile
     imageSize=<32 32>
     margin=<ltrb 0 0 0 0>
     outlineSettings {
      bUseBrushTransparency=false
      roundingType=HalfHeightRadius
      width=0
      color {colorUseRule=UseColor_Specified, specifiedColor=<rgba 0 0 0 0>}
      cornerRadii=<0 0 0 0>
     }
     tintColor {colorUseRule=UseColor_Specified, specifiedColor=<rgba 1 1 1 1>}
     uVRegion {bIsValid=false, max=<0 0>, min=<0 0>}
    }
   }
   slot {class=SizeBoxSlot}
  }
  InvalidationBox_Speaker {
   widgetClassPath=/Script/UMG.InvalidationBox
   parent=PC_PlayerNameInfo
   widget=InvalidationBox_Speaker
   slot {class=HorizontalBoxSlot, properties {horizontalAlignment=HAlign_Left}}
  }
  Overlay_0 {widgetClassPath=&Overlay, parent=InvalidationBox_Speaker, widget=Overlay_0, slot {class=PanelSlot}}
  SpeakerWidgetAnim {
   widgetClassPath=/Script/FortniteUI.FortVoiceChatStatusIconProxyAnim
   bInherited=true
   parent=Overlay_0
   widget=SpeakerWidgetAnim
   properties {renderOpacity=0.9, visibility=Hidden, desiredSize=<32 32>}
   slot {class=OverlaySlot, properties {verticalAlignment=VAlign_Center, padding=<ltrb 0 0 -12 0>}}
  }
  Mobile_PlayerNameInfo {
   widgetClassPath=&HorizontalBox
   bIsVariable=true
   parent=WidgetSwitcher
   widget=Mobile_PlayerNameInfo
   slot {class=WidgetSwitcherSlot}
  }
  Image_MarkerMobile {
   widgetClassPath=&Image
   bInherited=true
   bIsVariable=true
   parent=Mobile_PlayerNameInfo
   widget=Image_MarkerMobile
   properties {brush {
     bIsDynamicallyLoaded=false
     drawAs=Image
     imageType=NoImage
     mirroring=NoMirror
     resourceName=None
     resourceObject=/Game/Athena/HUD/Art/T_UI_Indicator_Diamond
     tiling=NoTile
     imageSize=<16 18>
     margin=<ltrb 0 0 0 0>
     outlineSettings {
      bUseBrushTransparency=false
      roundingType=HalfHeightRadius
      width=0
      color {colorUseRule=UseColor_Specified, specifiedColor=<rgba 0 0 0 0>}
      cornerRadii=<0 0 0 0>
     }
     tintColor {colorUseRule=UseColor_Specified, specifiedColor=<rgba 1 1 1 1>}
     uVRegion {bIsValid=false, max=<0 0>, min=<0 0>}
    }}
   slot {class=HorizontalBoxSlot, properties {horizontalAlignment=HAlign_Left, verticalAlignment=VAlign_Center, padding=<ltrb 5 5 5 5>}}
  }
  CommonTextBlock_PlayerNameMobile {
   widgetClassPath=&CommonTextBlock
   bInherited=true
   bIsVariable=true
   parent=Mobile_PlayerNameInfo
   widget=CommonTextBlock_PlayerNameMobile
   properties {
    text=Player Name
    font {
     bForceMonospaced=false
     bMaterialIsStencil=false
     fontMaterial=/Game/Athena/HUD/Materials/M_PlayerText_Gold_Inst
     fontObject=&BurbankSmall
     letterSpacing=0
     monospacedWidth=1
     size=14
     skewAmount=0
     typefaceFontName=Black
     outlineSettings {
      bApplyOutlineToDropShadows=true
      bMiteredCorners=false
      bSeparateFillAlpha=false
      outlineMaterial=null
      outlineSize=1
      outlineColor=<rgba 0 0 0 1>
     }
    }
    margin=<ltrb 0 0 2 -1>
    shadowColorAndOpacity=<rgba 0 0 0 1>
    shadowOffset=<1 1.5>
   }
   slot {class=HorizontalBoxSlot, properties {horizontalAlignment=HAlign_Left, verticalAlignment=VAlign_Center}}
  }
  TutorialHighlight {widgetClassPath=&TutorialHighlight}
 ]
}
