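# Squad list HUD widget that shows the seconds remaining on a reboot card as a single text countdown, hidden by default.
# SetupRemainingRebootCardTime stores the given time, starts a looping one-second timer and shows the text.
# UpdateRebootCardCountdown, apparently the timer's callback, decrements the stored value and displays it clamped at zero.
# FinishTimer collapses the text and clears the timer; nothing here calls it, so the countdown does not stop itself.
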WidgetBlueprint /BattleRoyaleGameUI/HUD/SquadList/WBP_BRGameUI_SquadDirective_RebootCardTimer
graphs [
 EventGraph {links [], nodes []}
 FinishTimer {nodes [
   FinishTimer() {is=title K2Node_FunctionEntry}
   > Set Visibility(Visibility=Collapsed, self=Get Text_Time() -> (Text_Time:Common Text Object Reference))
   > Clear and Invalidate Timer by Handle(Handle=Get TimerHandle() -> (TimerHandle:Timer Handle Structure))
  ]}
 SetupRemainingRebootCardTime {nodes [
   n1 = SetupRemainingRebootCardTime!() -> (then, RemainingTime:int) {is=title K2Node_FunctionEntry}
   > Set Remaining Time(Remaining Time:int=@n1) -> (Output_Get:int=0)
   > n2 = Set Timer by Event(Delegate=Create Event(), Time=1, bLooping=true)
   > Set TimerHandle(TimerHandle:Timer Handle Structure=@n2) -> (Output_Get:Timer Handle Structure)
   > Set Visibility(Visibility=SelfHitTestInvisible, self=Get Text_Time())
   > Return Node()
  ]}
 UpdateRebootCardCountdown {nodes [
   UpdateRebootCardCountdown() {is=title K2Node_FunctionEntry}
   > Decrement Int(Value=Get Remaining Time() -> (Remaining Time:int=0))
   n1 = "Max (Integer)"(A=Get Remaining Time() -> (Remaining Time=0), B=0)
   > "SetText (Text)"(self=Get Text_Time(), InText=Format Text!(Format="LOCTEXT(\"{x}s\")", x:Text="To Text (Integer)"(Value=@n1)) -> (Result))
  ]}
]
variables [
 Remaining Time {type=integer}
 TimerHandle {type=TimerHandle}
]
widgets {
 info {inheritedWidgetCount=0, parentClass=&UserWidget, rootWidgetClass=&Overlay, widgetCount=2}
 widgets [
  Overlay_0 {widgetClassPath=&Overlay, widget=Overlay_0}
  Text_Time {
   widgetClassPath=&CommonTextBlock
   bIsVariable=true
   parent=Overlay_0
   widget=Text_Time
   properties {
    minDesiredWidth=60
    text=268s
    visibility=Collapsed
    font {
     bForceMonospaced=false
     bMaterialIsStencil=false
     fontMaterial=/BattleRoyaleGameUI/HUD/Compass/Assets/M_Font_PP_White
     fontObject=&BurbankSmall
     letterSpacing=0
     monospacedWidth=1
     size=13.5
     skewAmount=0
     typefaceFontName=Black
     outlineSettings {
      bApplyOutlineToDropShadows=false
      bMiteredCorners=false
      bSeparateFillAlpha=true
      outlineMaterial=/BattleRoyaleGameUI/Shared/MI_UI_Font_SoftShadow_LargeText_CriticalGlow_Blue
      outlineSize=6
      outlineColor=<rgba 1 1 1 1>
     }
    }
    margin=<ltrb -4 -8 -4 -8>
    shadowOffset=<0 0>
   }
   slot {class=OverlaySlot, properties {verticalAlignment=VAlign_Center}}
  }
 ]
}
