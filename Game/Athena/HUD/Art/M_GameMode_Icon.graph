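# Translucent UI-domain material for a game mode icon in the Athena HUD.
# It exposes an Icon texture parameter, defaulting to a blank zero-alpha texture, and an IsVisible scalar parameter.
# The icon's alpha is multiplied by IsVisible, so the scalar can scale or zero out that alpha.
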Material /Game/Athena/HUD/Art/M_GameMode_Icon
graph {
 properties {blendMode=BLEND_Translucent, materialDomain=MD_UI, pixelDepthOffsetMode=PDOM_Legacy, usedShadingModels=Default Lit}
 nodes [Multiply(
   A=TextureSampleParameter2D() {parameterName=Icon, texture=/Game/Global/Textures/Default/Blanks/T_Blank_White_ZeroAlpha_sRGBColor}.A
   B=ScalarParameter() {parameterName=IsVisible}
  )]
}
