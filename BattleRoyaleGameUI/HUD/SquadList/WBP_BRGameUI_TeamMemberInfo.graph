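# Squad list HUD row for one teammate, showing portrait, name, health, shield and state icons in Battle Royale.
# Parent-class events drive DBNO, talking, resurrection and icon visuals through visibility toggles and material parameters.
# Playlist extension widgets route by slot enum into named panels, and a revive banner class resolves by tag and loads asynchronously.
# Touch layouts rearrange slots, and tagged playlists hide health bars.
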WidgetBlueprint /BattleRoyaleGameUI/HUD/SquadList/WBP_BRGameUI_TeamMemberInfo
animations [
 PulseMember {
  binding_count=1
  length_seconds=0.7204
  object=&self:PulseMember
  sequence {bindings [CommonBorderPulseBG {
     id=#1
     possessed_class=CommonBorder
     tracks [{
       type=MovieSceneColorTrack
       display_name=Brush Color
       sections [{
         type=MovieSceneColorSection
         end_frame=14
         start_frame=0
         channels [
          Color.R {type=MovieSceneScriptingFloatChannel, const [1,1]}
          Color.G {type=MovieSceneScriptingFloatChannel, const [1,1]}
          Color.B {type=MovieSceneScriptingFloatChannel, const [1,1]}
          Color.A {type=MovieSceneScriptingFloatChannel, keys [[0,1],[3.925,0],[4.363,1],[8.1317,0],[8.5697,1],[14.4083,0]]}
         ]
        }]
      }]
    }]}
 }
 DangerPulse {
  binding_count=1
  length_seconds=1.0105
  object=&self:DangerPulse
  sequence {bindings [CommonBorderPulseBG {
     id=#2
     possessed_class=CommonBorder
     tracks [{
       type=MovieSceneColorTrack
       display_name=Brush Color
       sections [{
         type=MovieSceneColorSection
         end_frame=20
         start_frame=0
         channels [
          Color.R {type=MovieSceneScriptingFloatChannel, const [0.5,1]}
          Color.G {type=MovieSceneScriptingFloatChannel, const [0,1]}
          Color.B {type=MovieSceneScriptingFloatChannel, const [0.01,1]}
          Color.A {type=MovieSceneScriptingFloatChannel, keys [[0,1],[20.209,0]]}
         ]
        }]
      }]
    }]}
 }
]
graphs [
 AddExtensionWidget {nodes [
   n1 = AddExtensionWidget!() -> (then, ExtensionWidgetToAdd:User Widget Object Reference, InputPin:EUIExtensionSlot Enum=Invalid) {is=title K2Node_FunctionEntry}
   > n2 = Switch on EUIExtensionSlot!(execute, Selection:EUIExtensionSlot Enum=@n1.InputPin) -> (
    Invalid:Exec
    Primary:Exec
    Secondary:Exec
    TopRightCorner:Exec
    GameInfoBox:Exec
    UpperCenter:Exec
    CrosshairRight:Exec
    MobileCrosshairRight:Exec
    UnderSquadInfo:Exec
    FullScreenMap:Exec
    BelowRespawnWidget:Exec
    BelowCompass:Exec
    UnderTeammateStatus:Exec
    ControllerBindingCallout:Exec
    AboveStormMessageSlot:Exec
    CustomMinigameCallouts:Exec
    UnderLocalPlayerInfo:Exec
    PlayerHealthbarOverlay:Exec
    UnderMinimapMessageContainer:Exec
    VehicleActionsSpacer:Exec
    Reticle:Exec
    PrioritizedContextualSlot:Exec
    RightOfTeammateStatus:Exec
    RightOfTeammateStatusBars:Exec
    TeammateStatusPortraitOverlay:Exec
    TeammateStatusPortraitUnderlay:Exec
    TeammateStatusPortraitReplacement:Exec
    TeammateStatusPortraitOverlay_IGM:Exec
    TeammateOverHealthbarOverlay:Exec
    TeammateOverShieldbarOverlay:Exec
    MobileHUDBottomRight:Exec
    InventoryScreenTab:Exec
    InventoryEquipSlot:Exec
    UnderneathTeammateStatusList:Exec
    FullScreenMapSquadEliminationsOverlay:Exec
    FullScreenMapStormInfo:Exec
    PrimaryQuickBarSlot:Exec
    MainMenuButtonListInZone:Exec
    MainMenuButtonListFrontEnd:Exec
    DynamicOverlayLobby:Exec
    PostMatchXPDetails_AccoladeList:Exec
    MobileHUDUnderMinimap:Exec
    SidebarPanels:Exec
    SpectatorRightOfPlayerStatus:Exec
    BehindMobileLocalPlayerInfo:Exec
    UnderMobileLocalPlayerInfo:Exec
    AboveMobileLocalPlayerInfo:Exec
    DynamicOverlayFrontEnd:Exec
    NPCStatusWidgetBarsPrefix:Exec
    NPCStatusWidgetAboveBars:Exec
    AbovePlayerIndicatorWidget:Exec
    FullScreenMapTab:Exec
    PostMatchScreenTab:Exec
    AboveBacchusEquippedItemReloadWidget:Exec
    UnderDefaultLobbyDetails:Exec
    DiscoveredAreaTextUnderlay:Exec
    LevelUpMobile:Exec
    ExpMobile:Exec
    InventoryGroupPanel:Exec
    InventoryItemInfoContents:Exec
    DiscoverInLobby:Exec
    InventoryItemInfoInspectPreview:Exec
    AboveQuestsMobileUI:Exec
    PostGamePlacementOverlayVictory:Exec
    TeammateStatusHealthShieldsReplacement:Exec
    GameInfoBelowMap:Exec
    LeftPostGameMobileOverlay:Exec
    LobbyAboveRankedDetails:Exec
    HUDPlayerActionAlertKillPlayerInfoExtension:Exec
    PostGamePlacementOverlayLosing:Exec
    AthenaInteractionIndicator:Exec
    HUDUnderReticle:Exec
    AthenaInventoryGroupMinorMedallionHeader:Exec
    AboveUpperCenterPrioritized:Exec
   ) {is=title K2Node_SwitchEnum}
   n3 = AddChildToOverlay!(
    execute=@n2.TeammateStatusPortraitUnderlay
    self:Overlay Object Reference=Get Underlay_PlayerPortraitExtensions() -> (Underlay_PlayerPortraitExtensions:Overlay Object Reference)
    Content:Widget Object Reference=Get ExtensionWidgetToAdd() -> (ExtensionWidgetToAdd:User Widget Object Reference)
   ) -> (then, ReturnValue:Overlay Slot Object Reference)
   > SetVerticalAlignment(self=@n3, InVerticalAlignment=VAlign_Fill)
   > SetHorizontalAlignment(self=@n3, InHorizontalAlignment=HAlign_Fill)
   n4 = AddChildToOverlay(
    execute=@n2.TeammateOverHealthbarOverlay
    self=Get Overlay_Healthbar() -> (Overlay_Healthbar:Overlay Object Reference)
    Content=Get ExtensionWidgetToAdd()
   )
   > SetVerticalAlignment(self=@n4, InVerticalAlignment=VAlign_Fill)
   > SetHorizontalAlignment(self=@n4, InHorizontalAlignment=HAlign_Fill)
   AddChild(
    execute=@n2.RightOfTeammateStatusBars
    self=Get RightOfTeammateExtensionSlot() -> (RightOfTeammateExtensionSlot:Panel Widget Object Reference)
    Content=Get ExtensionWidgetToAdd()
   )
   n5 = AddChildToVerticalBox(execute=@n2.UnderTeammateStatus, self=Get MainVertBox() -> (MainVertBox:Vertical Box Object Reference), Content=Get ExtensionWidgetToAdd())
   > SetVerticalAlignment(self:Vertical Box Slot Object Reference=@n5, InVerticalAlignment=VAlign_Center)
   > SetHorizontalAlignment(self:Vertical Box Slot Object Reference=@n5, InHorizontalAlignment=HAlign_Left)
   n6 = AddChildToOverlay(
    execute=@n2.TeammateOverShieldbarOverlay
    self=Get Overlay_ShieldBar() -> (Overlay_ShieldBar:Overlay Object Reference)
    Content=Get ExtensionWidgetToAdd()
   )
   > SetVerticalAlignment(self:Overlay Slot Object Reference=@n6, InVerticalAlignment=VAlign_Fill)
   > SetHorizontalAlignment(self:Overlay Slot Object Reference=@n6, InHorizontalAlignment=HAlign_Fill)
   n7 = AddChildToOverlay(
    execute=@n2.TeammateStatusPortraitOverlay
    self=Get Overlay_PlayerPortraitExtensions() -> (Overlay_PlayerPortraitExtensions:Overlay Object Reference)
    Content=Get ExtensionWidgetToAdd()
   )
   > SetVerticalAlignment(self=@n7, InVerticalAlignment=VAlign_Fill)
   > SetHorizontalAlignment(self=@n7, InHorizontalAlignment=HAlign_Fill)
   n8 = AddChildToOverlay(
    execute=@n2.TeammateStatusHealthShieldsReplacement
    self=Get Overlay_HealthShields() -> (Overlay_HealthShields:Common Visibility Switcher Object Reference)
    Content=Get ExtensionWidgetToAdd()
   )
   > SetVerticalAlignment(self=@n8, InVerticalAlignment=VAlign_Fill)
   > SetHorizontalAlignment(self=@n8, InHorizontalAlignment=HAlign_Fill)
   n9 = AddChildToOverlay(
    execute=@n2.TeammateStatusPortraitReplacement
    self=Get Overlay_PlayerPortraitReplacement() -> (Overlay_PlayerPortraitReplacement:Common Visibility Switcher Object Reference)
    Content=Get ExtensionWidgetToAdd()
   )
   > SetVerticalAlignment(self=@n9, InVerticalAlignment=VAlign_Fill)
   > SetHorizontalAlignment(self=@n9, InHorizontalAlignment=HAlign_Fill)
   n10 = AddChildToOverlay(
    execute=@n2.RightOfTeammateStatus
    self=Get Overlay_PlayerStateExtensions() -> (Overlay_PlayerStateExtensions:Overlay Object Reference)
    Content=Get ExtensionWidgetToAdd()
   )
   > SetVerticalAlignment(self=@n10, InVerticalAlignment=VAlign_Center)
   > SetHorizontalAlignment(self=@n10, InHorizontalAlignment=HAlign_Left)
  ]}
 EventGraph {nodes [
   Play Danger Pulse Anim() {is=title K2Node_CustomEvent}
   > PlayAnimationForward(InAnimation=Get DangerPulse() -> (DangerPulse:Widget Animation Object Reference))
   Play Pulse Anim() {is=title K2Node_CustomEvent}
   > PlayAnimationForward(InAnimation=Get PulseMember() -> (PulseMember:Widget Animation Object Reference))
   n1 = OverrideShowBackground!() -> (OutputDelegate, then, bShowBackground:bool=~) {is=title K2Node_CustomEvent}
   > SetBackgroundVisibility!(execute, self, bOverride:bool=true, bBackgroundVisible:bool=@n1.bShowBackground) -> (then)
   Play DBNO Flash() {is=title K2Node_CustomEvent}
   > n2 = Cast To FortGameStateAthena!(execute, Object=GetGameState!() -> (ReturnValue:Game State Base Object Reference)) -> (then, CastFailed, AsFort Game State Athena:Fort Game State Athena Object Reference) {is=title K2Node_DynamicCast}
   > n3 = Branch(Condition=ShouldDBNOTeamInfoFlash!(self:Fort Game State Athena Object Reference=@n2) -> (ReturnValue:bool))
   > Play Danger Pulse Anim() {is=title K2Node_CallFunction}
   > SetBackgroundVisibility(bOverride=false, bBackgroundVisible=true)
   Clear DBNO Flashing(execute=@n3.else)
   Clear DBNO Flashing() {is=title K2Node_CustomEvent}
   > Clear and Invalidate Timer by Handle(Handle=Get DBNOPulseTimer() -> (DBNOPulseTimer:Timer Handle Structure))
   > SetBackgroundVisibility(bOverride=false, bBackgroundVisible=false)
   n4 = Event Handle Playlist Loaded!() -> (OutputDelegate, then, PlaylistName:Name=None, PlaylistContextTags:"Gameplay Tag Container Structure (by ref)") {is=title K2Node_Event}
   > Branch(Condition=HasAnyTags!(
     TagContainer:"Gameplay Tag Container Structure (by ref)"=@n4.PlaylistContextTags
     OtherContainer:"Gameplay Tag Container Structure (by ref)"=Get RelevantPlaylistTags() -> (RelevantPlaylistTags:Gameplay Tag Container Structure)
     bExactMatch:bool=false
    ) -> (ReturnValue:bool))
   > SetVisibility(self=Get HorizontalBox_HealthShields() -> (HorizontalBox_HealthShields:Horizontal Box Object Reference), InVisibility=Collapsed)
   Event Construct()
   > n5 = Sequence()
   > SetVisibility(self=Get GameModeIcon() -> (GameModeIcon:Image Object Reference), InVisibility=Collapsed)
   > n6 = Branch(Condition=IsUsingTouch(PlayerController=GetOwningPlayer()))
   > Set RightOfTeammateExtensionSlot(RightOfTeammateExtensionSlot:Panel Widget Object Reference=Get HorizontalBox_Mobile_RightOfTeammateExtensions() -> (HorizontalBox_Mobile_RightOfTeammateExtensions:Horizontal Box Object Reference)) -> (Output_Get:Panel Widget Object Reference)
   n7 = Get HorizontalBox_Touch() -> (HorizontalBox_Touch:Horizontal Box Object Reference)
   > AddChild(self=@n7, Content=Get ScaleBox_HboxMobile_RightofTeammate() -> (ScaleBox_HboxMobile_RightofTeammate:Scale Box Object Reference))
   > AddChild(self=@n7, Content=Get TeamMemberStateIcon() -> (TeamMemberStateIcon:Image Object Reference))
   > AddChild(self=@n7, Content=Get GameModeIcon())
   > SetUserSpecifiedScale!(execute, self:Scale Box Object Reference=Get ScaleBox_HboxMobile_RightofTeammate(), InUserSpecifiedScale:float=2) -> (then)
   n8 = Set RightOfTeammateExtensionSlot(
    execute=@n6.else
    RightOfTeammateExtensionSlot=Get HorizontalBox_PlayerStateBarsExtensions() -> (HorizontalBox_PlayerStateBarsExtensions:Horizontal Box Object Reference)
   )
   > SetRenderScale!(execute, self:Widget Object Reference=@n8, Scale:Vector 2D Structure=(X=1,Y=1)) -> (then)
   K2Node_Composite_0 = CreateAliasedWidgets!(execute=@n5.then_1) -> (then:Exec) {is=title K2Node_Composite}
   n9 = Event Handle DBNO State Changed!() -> (OutputDelegate, then, bDBNO:bool) {is=title K2Node_Event}
   > n10 = Sequence!(execute) -> (then_0, then_1, then_2:Exec, then_3:Exec)
   > n11 = GetDynamicMaterial(self=Get Image_HealthBar() -> (Image_HealthBar:Image Object Reference))
   > SetScalarParameterValue(self=@n11, ParameterName=IsDBNO, Value="To Float (Boolean)"(InBool=@n9.bDBNO))
   n12 = Cast To FortGameStateAthena(execute=@n10.then_1, Object=GetGameState())
   > n13 = Branch(Condition=AND Boolean(A=ShouldDBNOTeamInfoFlash(self=@n12), B=@n9.bDBNO))
   > n14 = Set Timer by Event(Delegate=Create Event(), Time=1, bLooping=true)
   > Set DBNOPulseTimer(DBNOPulseTimer:Timer Handle Structure=@n14) -> (Output_Get:Timer Handle Structure)
   Clear DBNO Flashing(execute=[@n13.else, @n12.CastFailed]) {is=title K2Node_CallFunction}
   n15 = Branch(execute=@n10.then_2, Condition=Get bBypassPlayerStateBoxVisInvalidation() -> (bBypassPlayerStateBoxVisInvalidation:bool))
   UpdatePlayerStateBoxVisibility!(execute=[@n16.then, @n15.then], self, PlayerState:Fort Player State Athena Object Reference) -> (then)
   n16 = DelayUntilNextTick!(execute=@n15.else) -> (then)
   n17 = GetChildAt!(self:Panel Widget Object Reference=Get Overlay_ReviveBanner() -> (Overlay_ReviveBanner:Overlay Object Reference), Index:int=0) -> (ReturnValue:Widget Object Reference)
   Is Valid(exec=@n10.then_3, InputObject=@n17)
   > n18 = Cast To BPI_BRGameUI_TeamMemberInfo!(execute, Object=@n17) -> (then, CastFailed, AsBPI BRGame UI Team Member Info:BPI BRGame UI Team Member Info Interface) {is=title K2Node_DynamicCast}
   > SetTeamMemberPlayerState!(
    execute
    self:BPI BRGame UI Team Member Info Interface=@n18
    PlayerState:Player State Object Reference=Get WeakPSA() -> (WeakPSA:Fort Player State Athena Weak object)
   ) -> (then)
   n19 = Event Handle Talking Visual Display Changed!() -> (OutputDelegate, then, bTalkingVisualsActive:bool) {is=title K2Node_Event}
   > n20 = Set bIsTalking(bIsTalking:bool=@n19.bTalkingVisualsActive) -> (Output_Get:bool)
   > SetBackgroundVisibility(bOverride=false, bBackgroundVisible=@n20)
   n21 = Event Handle Resurrection State Changed!() -> (OutputDelegate, then, ResurrectionStateData:"Fort Resurrection UIData Structure (by ref)") {is=title K2Node_Event}
   > UpdateResurrectionState!(execute, self, FortResurrectionUIData:"Fort Resurrection UIData Structure (by ref)"=@n21.ResurrectionStateData) -> (then)
   n22 = Event Handle Game Mode Icon Changed!() -> (OutputDelegate, then, GameModeIcon:Texture 2D Object Reference) {is=title K2Node_Event}
   > n23 = Is Valid(InputObject=@n22.GameModeIcon)
   n24 = Get GameModeIcon()
   SetVisibility(execute=@n23.Is Not Valid, self=@n24, InVisibility=Collapsed)
   SetVisibility(execute=@n23.Is Valid, self=@n24, InVisibility=HitTestInvisible)
   > n25 = GetDynamicMaterial(self=Get GameModeIcon())
   > SetTextureParameterValue(self=@n25, ParameterName=Icon, Value=@n22.GameModeIcon)
   > SetScalarParameterValue(self=@n25, ParameterName=IsVisible, Value="To Float (Boolean)"(InBool=IsValid(Object=@n22.GameModeIcon)))
   n26 = Event Handle Playlist Extension Widget Created!() -> (OutputDelegate, then, Widget:User Widget Object Reference, SlotType:EUIExtensionSlot Enum=Invalid) {is=title K2Node_Event}
   > AddExtensionWidget!(execute, self, ExtensionWidgetToAdd:User Widget Object Reference=@n26.Widget, InputPin:EUIExtensionSlot Enum=@n26.SlotType) -> (then) {is=title K2Node_CallFunction}
   Event Collapse Squad Directive() {is=title K2Node_Event}
   > Set Visibility(Visibility=Collapsed, self=Get WBP_BRGameUI_SquadDirective() -> (WBP_BRGameUI_SquadDirective:WBP BRGame UI Squad Directive Object Reference))
   Event PreConstruct()
   > SetMobileStyles()
   n27 = Event Handle Team Member State Icon Changed!() -> (OutputDelegate, then, StateIcon:Texture 2D Object Reference) {is=title K2Node_Event}
   > n28 = Is Valid(InputObject=@n27.StateIcon)
   n29 = Get TeamMemberStateIcon()
   SetVisibility(execute=@n28.Is Not Valid, self=@n29, InVisibility=Collapsed)
   SetVisibility(execute=@n28.Is Valid, self=@n29, InVisibility=HitTestInvisible)
   > n30 = GetDynamicMaterial(self=@n29)
   > SetTextureParameterValue(self=@n30, ParameterName=Icon, Value=@n27.StateIcon)
   > SetScalarParameterValue(self=@n30, ParameterName=IsVisible, Value="To Float (Boolean)"(InBool=IsValid(Object=@n27.StateIcon)))
  ]}
 CreateAliasedWidgets {
  path=EventGraph.K2Node_Composite_0.CreateAliasedWidgets
  nodes [
   Inputs()
   n1 = GetSoftWidgetClassForAlias!(
    self:Fort UIData Configuration Context Object Reference=Get FortUIDataConfigurationContext!() -> (ReturnValue:Fort UIData Configuration Context Object Reference) {is=title K2Node_GetBlueprintContextEx}
    WidgetAlias:"Gameplay Tag Structure (by ref)"=Get ReviveBannerTag() -> (ReviveBannerTag:Gameplay Tag Structure)
   ) -> (ReturnValue:User Widget Soft Class Reference)
   > Branch(Condition=IsValidSoftClassReference!(SoftClassReference:"Object Soft Class Reference (by ref)"=@n1) -> (ReturnValue:bool))
   > n2 = CreateWidgetAsync!(
    execute:Exec
    UserWidgetSoftClass:User Widget Soft Class Reference=@n1
    OwningPlayer:Player Controller Object Reference
    bSuspendInputUntilComplete:bool=true
   ) -> (then:Exec, OnComplete:Exec, OnFailure:Exec, UserWidget:User Widget Object Reference) {is=title K2Node_AsyncAction}
   > Outputs!(then:Exec)
   n3 = AddChildToOverlay(execute=@n2.OnComplete, self=Get Overlay_ReviveBanner(), Content=@n2)
   > SetHorizontalAlignment(self=@n3, InHorizontalAlignment=HAlign_Fill)
   > SetVerticalAlignment(self=@n3, InVerticalAlignment=VAlign_Center)
   > Branch(Condition=Does Object Implement Interface!(
     TestObject:Object Reference=@n2
     Interface:Interface Interface=&/BPI_BRGameUI_TeamMemberInfo.BPI_BRGameUI_TeamMemberInfo_C
    ) -> (ReturnValue:bool))
   > n4 = Cast To BPI_BRGameUI_TeamMemberInfo(Object=@n2)
   > SetTeamMemberPlayerState(self=@n4, PlayerState=Get WeakPSA())
  ]
 }
 SetBackgroundVisibility {nodes [
   n1 = SetBackgroundVisibility!() -> (then, bOverride:bool=~, bBackgroundVisible:bool=~) {is=title K2Node_FunctionEntry}
   > n2 = Set bShowBackgroundOverridden(bShowBackgroundOverridden:bool=@n1) -> (Output_Get:bool)
   > n3 = Sequence!(execute) -> (then_0, then_1)
   > n4 = Branch(Condition=@n2)
   > Set bLocal_BackgroundVisible(bLocal_BackgroundVisible:bool=Get bBackgroundVisible() -> (bBackgroundVisible:bool)) -> (Output_Get:bool)
   n5 = Get DBNOPulseTimer()
   Set bLocal_BackgroundVisible(
    execute=@n4.else
    bLocal_BackgroundVisible=OR Boolean!(
     A=Get bBackgroundVisible()
     B=Get bIsTalking() -> (bIsTalking:bool)
     C:bool=Is Valid Timer Handle!(Handle:Timer Handle Structure=@n5) -> (ReturnValue:bool)
    ) -> (ReturnValue)
   )
   SetVisibility(
    execute=@n3.then_1
    self=Get BG_Active() -> (BG_Active:Image Object Reference)
    InVisibility=Select(Option 0=Hidden, Option 1=SelfHitTestInvisible, Index=Get bLocal_BackgroundVisible() -> (bLocal_BackgroundVisible:bool))
   )
  ]}
 SetMobileStyles {nodes [
   SetMobileStyles() {is=title K2Node_FunctionEntry}
   n1 = GetOwningPlayer()
   > Branch(Condition=AND Boolean(
     A=IsUsingCompactUI()
     B=IsUsingInputType!(InputType:ECommonInputType Enum=Touch, PlayerController:Player Controller Object Reference=@n1) -> (ReturnValue:bool)
    ))
   > SetRenderScale(self=Get Overlay_PlayerPortraitReplacement(), Scale=(X=0.9,Y=0.9))
   n2 = Get WBP_BRGameUI_SquadDirective()
   > SetPadding!(
    execute
    self:User Widget Object Reference=@n2
    InPadding_Left:float=0
    InPadding_Top:float=0
    InPadding_Right:float=0
    InPadding_Bottom:float=0
   ) -> (then)
   > SetHorizontalAlignment(self=SlotAsOverlaySlot(Widget=@n2), InHorizontalAlignment=HAlign_Left)
  ]}
 SetRebootCardTime {nodes [
   n1 = SetRebootCardTime!() -> (then, RemainingTime:int) {is=title K2Node_FunctionEntry}
   > SetupRemainingRebootCardTime!(
    execute
    self:WBP BRGame UI Squad Directive Reboot Card Timer Object Reference=Get WBP_SquadDirective_RebootCardTimer() -> (WBP_SquadDirective_RebootCardTimer:WBP BRGame UI Squad Directive Reboot Card Timer Object Reference)
    RemainingTime:int=@n1
   ) -> (then)
  ]}
 UpdatePlayerStateBoxVisibility {nodes [
   UpdatePlayerStateBoxVisibility!() -> (then, PlayerState:Fort Player State Athena Object Reference) {is=title K2Node_FunctionEntry}
   > n1 = Branch(Condition=Get bBypassPlayerStateBoxVisInvalidation())
   n2 = Get Image_DeadIndicator() -> (Image_DeadIndicator:Image Object Reference)
   n3 = Get Image_ResurrectChip() -> (Image_ResurrectChip:Image Object Reference)
   n4 = Branch(execute=[@n19.then, @n1.then], Condition=OR Boolean!(A=IsVisible(self=@n2), B=IsVisible(self=@n3)) -> (ReturnValue))
   FinishTimer(execute=[@n4.then, @n6.then], self:WBP BRGame UI Squad Directive Reboot Card Timer Object Reference=Get WBP_SquadDirective_RebootCardTimer())
   > Unbind Event from Received Despawn Time(self:WBP BRGame UI Squad Directive Object Reference=Get WBP_BRGameUI_SquadDirective(), Delegate=Create Event()) {is=title K2Node_RemoveDelegate}
   n5 = Get PlayerState() -> (PlayerState:Fort Player State Athena Object Reference)
   Is Valid(exec=@n4.else, InputObject=@n5)
   > n6 = Branch(Condition=NOT Boolean(A=IsPlayerDead!(self:Fort Player State Zone Object Reference=@n5) -> (ReturnValue:bool)))
   n7 = IsVisible(self=Get DBNOStateWidget() -> (DBNOStateWidget:Athena Team Member DBNOState Object Reference))
   n8 = IsVisible(self=Get Image_DeadIndicator())
   n9 = IsVisible(self=Get Image_DisconnectedIndicator() -> (Image_DisconnectedIndicator:Image Object Reference))
   n10 = IsVisible(self=Get Image_ResurrectChip())
   n11 = IsVisible(self=Get Image_ResurrectChip_Uncollected() -> (Image_ResurrectChip_Uncollected:Image Object Reference))
   SetVisibility(
    execute=@n1.else
    self=Get SizeBox_PlayerStates() -> (SizeBox_PlayerStates:Size Box Object Reference)
    InVisibility=Select(Option 0=Collapsed, Option 1=HitTestInvisible, Index=OR Boolean!(A=@n10, B=@n11, C:bool=@n9, D:bool=@n8, E:bool=@n7) -> (ReturnValue))
   )
   n12 = IsVisible(self=Get DBNOStateWidget())
   > SetVisibility(self=Get DBNOGlow() -> (DBNOGlow:Image Object Reference), InVisibility=Select(Option 0=Collapsed, Option 1=HitTestInvisible, Index=@n12))
   > n13 = GetDynamicMaterial(self=Get Backing() -> (Backing:Image Object Reference))
   > SetScalarParameterValue(self=@n13, ParameterName=IsRed, Value=Select(Option 0:float=0, Option 1:float=1, Index=@n12) -> (ReturnValue:float))
   > n14 = GetDynamicMaterial(self=Get Backing())
   n15 = IsVisible(self=Get Image_DeadIndicator())
   n16 = IsVisible(self=Get Image_ResurrectChip())
   n17 = IsVisible(self=Get Image_ResurrectChip_Uncollected())
   n18 = IsVisible(self=Get Image_DisconnectedIndicator())
   > n19 = SetScalarParameterValue(self=@n14, ParameterName=IsDead, Value=Select(Option 0=0, Option 1=1, Index=OR Boolean!(A=@n15, B=@n18, C=@n16, D=@n17) -> (ReturnValue)))
  ]}
 UpdateResurrectionState {nodes [
   UpdateResurrectionState!() -> (then, FortResurrectionUIData:"Fort Resurrection UIData Structure (by ref)") {is=title K2Node_FunctionEntry}
   > n1 = Sequence!(execute) -> (then_0, then_1, then_2:Exec)
   n2 = Get!() -> (
    FortResurrectionUIData_bResurrectionChipAvailable:bool=~
    FortResurrectionUIData_bResurrectionChipPickedUp:bool=~
    FortResurrectionUIData_bResurrectionChipRebooting:bool=~
    FortResurrectionUIData_SquadDirectiveUpdateType:EAthenaSquadListUpdateType Enum
    FortResurrectionUIData_PlayerState:Fort Player State Athena Object Reference
   )
   > Delta Gate Wildcard!(
    execute:Exec
    Variable:"EAthenaSquadListUpdateType Enum (by ref)"=Get PreviousSquadDirective() -> (PreviousSquadDirective:EAthenaSquadListUpdateType Enum=None)
    New Value:EAthenaSquadListUpdateType Enum=@n2.FortResurrectionUIData_SquadDirectiveUpdateType
   ) -> (Success:Exec, Updated Variable:"EAthenaSquadListUpdateType Enum (by ref)", Previous Variable:EAthenaSquadListUpdateType Enum) {is=title K2Node_MacroInstance}
   > EnableSquadDirective!(
    execute
    self:WBP BRGame UI Squad Directive Object Reference=Get WBP_BRGameUI_SquadDirective()
    EnableSquadDirective:bool="Not Equal (Enum)"!(A:EAthenaSquadListUpdateType Enum=@n2.FortResurrectionUIData_SquadDirectiveUpdateType, B:EAthenaSquadListUpdateType Enum=None) -> (ReturnValue:bool)
   ) -> (then)
   > Bind Event to Received Despawn Time(self:WBP BRGame UI Squad Directive Object Reference=Get WBP_BRGameUI_SquadDirective(), Delegate=Create Event()) {is=title K2Node_AddDelegate}
   SetVisibility(execute=@n1.then_1, self=Get Image_ResurrectChip_Uncollected(), InVisibility=Collapsed)
   n3 = Get() -> (
    FortResurrectionUIData_bResurrectionChipAvailable=~
    FortResurrectionUIData_bResurrectionChipPickedUp=~
    FortResurrectionUIData_bResurrectionChipRebooting=~
   )
   n4 = Get Image_ResurrectChip()
   > Branch(Condition=AND Boolean(A=IsVisible(self=@n4), B=NOT Boolean(A=@n3.FortResurrectionUIData_bResurrectionChipPickedUp)))
   > SetVisibility(self=Get Image_ResurrectChip(), InVisibility=Collapsed)
   > SetVisibility(self=Get Image_ResurrectChip_Uncollected(), InVisibility=SelfHitTestInvisible)
   n5 = Get() -> (
    FortResurrectionUIData_bResurrectionChipAvailable=~
    FortResurrectionUIData_bResurrectionChipPickedUp=~
    FortResurrectionUIData_bResurrectionChipRebooting=~
   )
   UpdatePlayerStateBoxVisibility(execute=@n1.then_2, PlayerState=@n5.FortResurrectionUIData_PlayerState) {is=title K2Node_CallFunction}
   n6 = GetPlayerController!(self:Player State Object Reference=@n5.FortResurrectionUIData_PlayerState) -> (ReturnValue:Player Controller Object Reference)
   > n7 = Is Valid(InputObject=@n6)
   > n8 = Is Valid(InputObject=GetSpectatorPawn!(self:Player Controller Object Reference=@n6) -> (ReturnValue:Spectator Pawn Object Reference))
   ForceUpdateSquadListDisplay!(
    execute=[@n8.Is Not Valid, @n7.Is Not Valid]
    self:WBP BRGame UI Squad Directive Object Reference=Get WBP_BRGameUI_SquadDirective()
    Update Type:EAthenaSquadListUpdateType Enum=@n5.FortResurrectionUIData_SquadDirectiveUpdateType
    PS:Fort Player State Athena Object Reference=@n5.FortResurrectionUIData_PlayerState
   ) -> (then)
  ]}
]
variables [
 bBypassPlayerStateBoxVisInvalidation {type=boolean}
 bIsTalking {type=boolean}
 bShowBackgroundOverridden {type=boolean}
 DBNOPulseTimer {type=TimerHandle}
 PreviousSquadDirective {type=EAthenaSquadListUpdateType, default=None}
 RelevantPlaylistTags {type=GameplayTagContainer, default {gameplayTags [{tagName=Athena.Playlist.Papaya}]}}
 ReviveBannerTag {type=GameplayTag, default {tagName=HUD.Athena.TeamMemberInfo.Revive}}
 RightOfTeammateExtensionSlot {type=PanelWidget, default=null}
]
widgets {
 info {inheritedWidgetCount=17, parentClass=/Script/FortniteUI.AthenaPlayerInfoBase, rootWidgetClass=&Overlay, widgetCount=56}
 widgets [
  Overlay_TeamMemberInfo {widgetClassPath=&Overlay, bInherited=true, widget=Overlay_TeamMemberInfo}
  FortVisualAttachment_0 {
   widgetClassPath=&FortVisualAttachment
   parent=Overlay_TeamMemberInfo
   widget=FortVisualAttachment_0
   slot {class=OverlaySlot, properties {horizontalAlignment=HAlign_Fill, verticalAlignment=VAlign_Fill}}
  }
  Overlay_2 {widgetClassPath=&Overlay, parent=FortVisualAttachment_0, widget=Overlay_2, slot {class=SizeBoxSlot}}
  DBNOGlow {
   widgetClassPath=&Image
   bIsVariable=true
   parent=Overlay_2
   widget=DBNOGlow
   properties {
    visibility=Collapsed
    brush {
     bIsDynamicallyLoaded=false
     drawAs=Image
     imageType=FullColor
     mirroring=NoMirror
     resourceName=None
     resourceObject=/BattleRoyaleGameUI/Shared/FX/M_UI_ElimGlow
     tiling=NoTile
     imageSize=<48 48>
     margin=<ltrb 0 0 0 0>
     outlineSettings {
      bUseBrushTransparency=false
      roundingType=HalfHeightRadius
      width=0
      color {colorUseRule=UseColor_Specified, specifiedColor=<rgba 0 0 0 0>}
      cornerRadii=<0 0 0 0>
     }
     tintColor {colorUseRule=UseColor_Specified, specifiedColor=<rgba 1 1 1 1>}
     uVRegion {bIsValid=false, max=<0 0>, min=<0 0>}
    }
    renderTransform {angle=0, scale=<3 2.5>, shear=<0 0>, translation=<-16 0>}
   }
   slot {class=OverlaySlot, properties {verticalAlignment=VAlign_Fill}}
  }
  Backing {
   widgetClassPath=&Image
   bIsVariable=true
   parent=Overlay_2
   widget=Backing
   properties {
    visibility=SelfHitTestInvisible
    brush {
     bIsDynamicallyLoaded=false
     drawAs=Image
     imageType=FullColor
     mirroring=NoMirror
     resourceName=None
     resourceObject=/BattleRoyaleGameUI/HUD/SquadList/Assets/M_UI_SquadList_Backing
     tiling=NoTile
     imageSize=<384 48>
     margin=<ltrb 0 0 0 0>
     outlineSettings {
      bUseBrushTransparency=false
      roundingType=HalfHeightRadius
      width=0
      color {colorUseRule=UseColor_Specified, specifiedColor=<rgba 0 0 0 0>}
      cornerRadii=<0 0 0 0>
     }
     tintColor {colorUseRule=UseColor_Specified, specifiedColor=<rgba 1 1 1 1>}
     uVRegion {bIsValid=false, max=<0 0>, min=<0 0>}
    }
   }
   slot {class=OverlaySlot}
  }
  BG_Active {
   widgetClassPath=&Image
   bIsVariable=true
   parent=Overlay_TeamMemberInfo
   widget=BG_Active
   properties {
    renderOpacity=0
    visibility=SelfHitTestInvisible
    brush {
     bIsDynamicallyLoaded=false
     drawAs=Image
     imageType=NoImage
     mirroring=NoMirror
     resourceName=None
     resourceObject=/Game/Athena/HUD/LTM/Wax/Materials/M_UI_TextGlow
     tiling=NoTile
     imageSize=<32 32>
     margin=<ltrb 0 0 0 0>
     outlineSettings {
      bUseBrushTransparency=false
      roundingType=HalfHeightRadius
      width=0
      color {colorUseRule=UseColor_Specified, specifiedColor=<rgba 0 0 0 0>}
      cornerRadii=<0 0 0 0>
     }
     tintColor {colorUseRule=UseColor_Specified, specifiedColor=<rgba 0 0.4564 1 1>}
     uVRegion {bIsValid=false, max=<0 0>, min=<0 0>}
    }
    colorAndOpacity=<rgba 0 0.456 1 1>
    renderTransform {angle=0, scale=<1.4 0.7>, shear=<0 0>, translation=<0 -9>}
   }
   slot {class=OverlaySlot, properties {horizontalAlignment=HAlign_Fill, verticalAlignment=VAlign_Fill}}
  }
  InvalidationBox_94 {widgetClassPath=/Script/UMG.InvalidationBox, parent=Overlay_TeamMemberInfo, widget=InvalidationBox_94, slot {class=OverlaySlot}}
  CommonBorderBackground {
   widgetClassPath=&CommonBorder
   parent=InvalidationBox_94
   widget=CommonBorderBackground
   properties {
    background {
     bIsDynamicallyLoaded=false
     drawAs=Image
     imageType=NoImage
     mirroring=NoMirror
     resourceName=None
     resourceObject=null
     tiling=NoTile
     imageSize=<32 32>
     margin=<ltrb 0 0 0 0>
     outlineSettings {
      bUseBrushTransparency=false
      roundingType=HalfHeightRadius
      width=0
      color {colorUseRule=UseColor_Specified, specifiedColor=<rgba 0 0 0 0>}
      cornerRadii=<0 0 0 0>
     }
     tintColor {colorUseRule=UseColor_Specified, specifiedColor=<rgba 0.5602 1 0.0408 0>}
     uVRegion {bIsValid=false, max=<0 0>, min=<0 0>}
    }
    brushColor=<rgba 1 1 1 0>
    padding=<ltrb 0 0 0 0>
   }
   slot {class=PanelSlot}
  }
  CommonBorderPulseBG {
   widgetClassPath=&CommonBorder
   parent=CommonBorderBackground
   widget=CommonBorderPulseBG
   properties {
    background {
     bIsDynamicallyLoaded=false
     drawAs=Image
     imageType=NoImage
     mirroring=NoMirror
     resourceName=None
     resourceObject=null
     tiling=NoTile
     imageSize=<32 32>
     margin=<ltrb 0 0 0 0>
     outlineSettings {
      bUseBrushTransparency=false
      roundingType=HalfHeightRadius
      width=0
      color {colorUseRule=UseColor_Specified, specifiedColor=<rgba 0 0 0 0>}
      cornerRadii=<0 0 0 0>
     }
     tintColor {colorUseRule=UseColor_Specified, specifiedColor=<rgba 1 1 1 0.8>}
     uVRegion {bIsValid=false, max=<0 0>, min=<0 0>}
    }
    brushColor=<rgba 1 1 1 0>
    padding=<ltrb 0 0 0 0>
   }
   slot {class=BorderSlot, properties {padding=<ltrb 0 0 0 0>}}
  }
  MainVertBox {
   widgetClassPath=&VerticalBox
   bIsVariable=true
   parent=CommonBorderPulseBG
   widget=MainVertBox
   slot {class=BorderSlot, properties {padding=<ltrb 0 0 0 0>}}
  }
  Overlay_0 {widgetClassPath=&Overlay, parent=MainVertBox, widget=Overlay_0, slot {class=VerticalBoxSlot}}
  HorizontalBox_0 {widgetClassPath=&HorizontalBox, parent=Overlay_0, widget=HorizontalBox_0, slot {class=OverlaySlot}}
  Overlay_PlayerPortraitExtensions {
   widgetClassPath=&Overlay
   bIsVariable=true
   parent=HorizontalBox_0
   widget=Overlay_PlayerPortraitExtensions
   slot {class=HorizontalBoxSlot, properties {horizontalAlignment=HAlign_Left, verticalAlignment=VAlign_Center}}
  }
  Underlay_PlayerPortraitExtensions {
   widgetClassPath=&Overlay
   bIsVariable=true
   parent=Overlay_PlayerPortraitExtensions
   widget=Underlay_PlayerPortraitExtensions
   slot {class=OverlaySlot, properties {horizontalAlignment=HAlign_Fill, verticalAlignment=VAlign_Fill}}
  }
  Overlay_PlayerPortraitReplacement {
   widgetClassPath=&CommonVisibilitySwitcher
   bIsVariable=true
   parent=Overlay_PlayerPortraitExtensions
   widget=Overlay_PlayerPortraitReplacement
   slot {class=OverlaySlot}
  }
  WidgetSwitcher_Avatar {
   widgetClassPath=&CommonWidgetSwitcherLegacy
   bInherited=true
   bIsVariable=true
   parent=Overlay_PlayerPortraitReplacement
   widget=WidgetSwitcher_Avatar
   slot {class=CommonVisibilitySwitcherSlot}
  }
  Icon_AvatarSocial {
   widgetClassPath=/Script/FortniteUI.FortSocialAvatarIcon
   bInherited=true
   parent=WidgetSwitcher_Avatar
   widget=Icon_AvatarSocial
   properties {
    materialTextureParamName=SmallPreview
    brush {
     bIsDynamicallyLoaded=false
     drawAs=Image
     imageType=FullColor
     mirroring=NoMirror
     resourceName=None
     resourceObject=/BattleRoyaleGameUI/HUD/SquadList/Assets/M_UI_SquadsList_Portrait
     tiling=NoTile
     imageSize=<48 48>
     margin=<ltrb 0 0 0 0>
     outlineSettings {
      bUseBrushTransparency=false
      roundingType=HalfHeightRadius
      width=0
      color {colorUseRule=UseColor_Specified, specifiedColor=<rgba 0 0 0 0>}
      cornerRadii=<0 0 0 0>
     }
     tintColor {colorUseRule=UseColor_Specified, specifiedColor=<rgba 1 1 1 1>}
     uVRegion {bIsValid=false, max=<0 0>, min=<0 0>}
    }
   }
   slot {class=WidgetSwitcherSlot}
  }
  Image_AvatarOverride {
   widgetClassPath=&Image
   bInherited=true
   bIsVariable=true
   parent=WidgetSwitcher_Avatar
   widget=Image_AvatarOverride
   properties {brush {
     bIsDynamicallyLoaded=false
     drawAs=Image
     imageType=NoImage
     mirroring=NoMirror
     resourceName=None
     resourceObject=null
     tiling=NoTile
     imageSize=<48 48>
     margin=<ltrb 0 0 0 0>
     outlineSettings {
      bUseBrushTransparency=false
      roundingType=HalfHeightRadius
      width=0
      color {colorUseRule=UseColor_Specified, specifiedColor=<rgba 0 0 0 0>}
      cornerRadii=<0 0 0 0>
     }
     tintColor {colorUseRule=UseColor_Specified, specifiedColor=<rgba 1 1 1 1>}
     uVRegion {bIsValid=false, max=<0 0>, min=<0 0>}
    }}
   slot {class=WidgetSwitcherSlot}
  }
  Player {
   widgetClassPath=&VerticalBox
   parent=HorizontalBox_0
   widget=Player
   slot {class=HorizontalBoxSlot, properties {horizontalAlignment=HAlign_Left, verticalAlignment=VAlign_Top, padding=<ltrb 4 0 0 0>}}
  }
  HorizontalBox_3 {
   widgetClassPath=&HorizontalBox
   parent=Player
   widget=HorizontalBox_3
   slot {class=VerticalBoxSlot, properties {horizontalAlignment=HAlign_Left}}
  }
  AthenaPlayerNameInfo {
   widgetClassPath=&/WBP_BRGameUI_PlayerNameInfo
   bInherited=true
   bIsVariable=true
   parent=HorizontalBox_3
   widget=AthenaPlayerNameInfo
   slot {class=HorizontalBoxSlot, properties {verticalAlignment=VAlign_Center}}
  }
  Overlay_HealthShields {
   widgetClassPath=&CommonVisibilitySwitcher
   bIsVariable=true
   parent=Player
   widget=Overlay_HealthShields
   slot {class=VerticalBoxSlot, properties {horizontalAlignment=HAlign_Left, verticalAlignment=VAlign_Center}}
  }
  HorizontalBox_HealthShields {
   widgetClassPath=&HorizontalBox
   bIsVariable=true
   parent=Overlay_HealthShields
   widget=HorizontalBox_HealthShields
   properties {flowDirectionPreference=Culture}
   slot {class=CommonVisibilitySwitcherSlot, properties {verticalAlignment=VAlign_Top}}
  }
  SizeBox_3 {
   widgetClassPath=&SizeBox
   parent=HorizontalBox_HealthShields
   widget=SizeBox_3
   properties {bOverride_MinDesiredWidth=true, minDesiredWidth=212, widthOverride=200}
   slot {class=HorizontalBoxSlot, properties {verticalAlignment=VAlign_Top}}
  }
  InvalidationBox_HealthAndShield {widgetClassPath=&InvalidationBox, parent=SizeBox_3, widget=InvalidationBox_HealthAndShield, slot {class=SizeBoxSlot}}
  Overlay_HealthAndShieldInfo {
   widgetClassPath=&Overlay
   bInherited=true
   bIsVariable=true
   parent=InvalidationBox_HealthAndShield
   widget=Overlay_HealthAndShieldInfo
   slot {class=PanelSlot}
  }
  VerticalBox_0 {
   widgetClassPath=&VerticalBox
   parent=Overlay_HealthAndShieldInfo
   widget=VerticalBox_0
   slot {class=OverlaySlot, properties {horizontalAlignment=HAlign_Fill}}
  }
  SizeBox_ShieldBar {
   widgetClassPath=&SizeBox
   parent=VerticalBox_0
   widget=SizeBox_ShieldBar
   properties {bOverride_HeightOverride=true, heightOverride=6}
   slot {class=VerticalBoxSlot}
  }
  Overlay_ShieldBar {widgetClassPath=&Overlay, bIsVariable=true, parent=SizeBox_ShieldBar, widget=Overlay_ShieldBar, slot {class=SizeBoxSlot}}
  Image_ShieldBar {
   widgetClassPath=&Image
   bInherited=true
   bIsVariable=true
   parent=Overlay_ShieldBar
   widget=Image_ShieldBar
   properties {brush {
     bIsDynamicallyLoaded=false
     drawAs=Image
     imageType=FullColor
     mirroring=NoMirror
     resourceName=None
     resourceObject=/BattleRoyaleGameUI/HUD/PlayerHitPointInfo/Assets/MI_UI_ShieldBar_Squadmate
     tiling=NoTile
     imageSize=<16 8>
     margin=<ltrb 0 0 0 0>
     outlineSettings {
      bUseBrushTransparency=false
      roundingType=HalfHeightRadius
      width=0
      color {colorUseRule=UseColor_Specified, specifiedColor=<rgba 0 0 0 0>}
      cornerRadii=<0 0 0 0>
     }
     tintColor {colorUseRule=UseColor_Specified, specifiedColor=<rgba 1 1 1 1>}
     uVRegion {bIsValid=false, max=<0 0>, min=<0 0>}
    }}
   slot {class=OverlaySlot, properties {horizontalAlignment=HAlign_Fill, verticalAlignment=VAlign_Fill}}
  }
  SizeBox_HealthBar {
   widgetClassPath=&SizeBox
   parent=VerticalBox_0
   widget=SizeBox_HealthBar
   properties {bOverride_MaxDesiredHeight=true, heightOverride=10, maxDesiredHeight=10}
   slot {class=VerticalBoxSlot, properties {verticalAlignment=VAlign_Top, padding=<ltrb 0 2 0 0>}}
  }
  Overlay_Healthbar {widgetClassPath=&Overlay, bIsVariable=true, parent=SizeBox_HealthBar, widget=Overlay_Healthbar, slot {class=SizeBoxSlot}}
  Image_HealthBar {
   widgetClassPath=&Image
   bInherited=true
   bIsVariable=true
   parent=Overlay_Healthbar
   widget=Image_HealthBar
   properties {brush {
     bIsDynamicallyLoaded=false
     drawAs=Image
     imageType=FullColor
     mirroring=NoMirror
     resourceName=None
     resourceObject=/BattleRoyaleGameUI/HUD/PlayerHitPointInfo/Assets/MI_UI_HealthBar_Squadmate
     tiling=NoTile
     imageSize=<16 12>
     margin=<ltrb 0 0 0 0>
     outlineSettings {
      bUseBrushTransparency=false
      roundingType=HalfHeightRadius
      width=0
      color {colorUseRule=UseColor_Specified, specifiedColor=<rgba 0 0 0 0>}
      cornerRadii=<0 0 0 0>
     }
     tintColor {colorUseRule=UseColor_Specified, specifiedColor=<rgba 1 1 1 1>}
     uVRegion {bIsValid=false, max=<0 0>, min=<0 0>}
    }}
   slot {class=OverlaySlot, properties {horizontalAlignment=HAlign_Fill, verticalAlignment=VAlign_Fill}}
  }
  Content_StormCN {
   widgetClassPath=&SizeBox
   bInherited=true
   parent=VerticalBox_0
   widget=Content_StormCN
   slot {class=VerticalBoxSlot, properties {padding=<ltrb 0 2 0 0>}}
  }
  Image_StormCn {
   widgetClassPath=&Image
   bInherited=true
   bIsVariable=true
   parent=Content_StormCN
   widget=Image_StormCn
   properties {
    visibility=Collapsed
    brush {
     bIsDynamicallyLoaded=false
     drawAs=Image
     imageType=NoImage
     mirroring=NoMirror
     resourceName=None
     resourceObject=/Game/Athena/UI/GAPP/M_ProgressBarCN_Base_Inst
     tiling=NoTile
     imageSize=<16 4>
     margin=<ltrb 0 0 0 0>
     outlineSettings {
      bUseBrushTransparency=false
      roundingType=HalfHeightRadius
      width=0
      color {colorUseRule=UseColor_Specified, specifiedColor=<rgba 0 0 0 0>}
      cornerRadii=<0 0 0 0>
     }
     tintColor {colorUseRule=UseColor_Specified, specifiedColor=<rgba 1 1 1 1>}
     uVRegion {bIsValid=false, max=<0 0>, min=<0 0>}
    }
   }
   slot {class=SizeBoxSlot}
  }
  Image_ResurrectionChipExpirationBar {
   widgetClassPath=&Image
   bInherited=true
   bIsVariable=true
   parent=Overlay_HealthAndShieldInfo
   widget=Image_ResurrectionChipExpirationBar
   properties {
    visibility=Collapsed
    brush {
     bIsDynamicallyLoaded=false
     drawAs=Image
     imageType=NoImage
     mirroring=NoMirror
     resourceName=None
     resourceObject=/Game/Athena/HUD/Team/MI_Player_RebootBar
     tiling=NoTile
     imageSize=<24 18>
     margin=<ltrb 0 0 0 0>
     outlineSettings {
      bUseBrushTransparency=false
      roundingType=HalfHeightRadius
      width=0
      color {colorUseRule=UseColor_Specified, specifiedColor=<rgba 0 0 0 0>}
      cornerRadii=<0 0 0 0>
     }
     tintColor {colorUseRule=UseColor_Specified, specifiedColor=<rgba 1 1 1 1>}
     uVRegion {bIsValid=false, max=<0 0>, min=<0 0>}
    }
   }
   slot {class=OverlaySlot, properties {horizontalAlignment=HAlign_Fill}}
  }
  SizeBox_PlayerStates {
   widgetClassPath=&SizeBox
   bIsVariable=true
   parent=HorizontalBox_HealthShields
   widget=SizeBox_PlayerStates
   properties {
    bOverride_HeightOverride=true
    bOverride_WidthOverride=true
    heightOverride=20
    maxDesiredHeight=28
    maxDesiredWidth=28
    widthOverride=20
    renderTransform {angle=0, scale=<1.25 1.25>, shear=<0 0>, translation=<0 -2>}
   }
   slot {class=HorizontalBoxSlot, properties {horizontalAlignment=HAlign_Left, verticalAlignment=VAlign_Top, padding=<ltrb 6 0 0 0>}}
  }
  InvalidationBox_PlayerStates {widgetClassPath=&InvalidationBox, parent=SizeBox_PlayerStates, widget=InvalidationBox_PlayerStates, slot {class=SizeBoxSlot}}
  Overlay_1 {widgetClassPath=&Overlay, parent=InvalidationBox_PlayerStates, widget=Overlay_1, slot {class=PanelSlot}}
  DBNOStateWidget {
   widgetClassPath=/Game/Athena/HUD/Team/AthenaTeamMemberDBNOState
   bInherited=true
   bIsVariable=true
   parent=Overlay_1
   widget=DBNOStateWidget
   properties {visibility=Collapsed}
   slot {class=OverlaySlot, properties {horizontalAlignment=HAlign_Fill, verticalAlignment=VAlign_Fill}}
  }
  Image_DeadIndicator {
   widgetClassPath=&Image
   bInherited=true
   bIsVariable=true
   parent=Overlay_1
   widget=Image_DeadIndicator
   properties {
    visibility=Collapsed
    brush {
     bIsDynamicallyLoaded=false
     drawAs=Image
     imageType=NoImage
     mirroring=NoMirror
     resourceName=None
     resourceObject=/Game/Athena/HUD/Art/squadArt/squad_Eliminated
     tiling=NoTile
     imageSize=<64 64>
     margin=<ltrb 0 0 0 0>
     outlineSettings {
      bUseBrushTransparency=false
      roundingType=HalfHeightRadius
      width=0
      color {colorUseRule=UseColor_Specified, specifiedColor=<rgba 0 0 0 0>}
      cornerRadii=<0 0 0 0>
     }
     tintColor {colorUseRule=UseColor_Specified, specifiedColor=<rgba 1 0.0168 0.0953 1>}
     uVRegion {bIsValid=false, max=<0 0>, min=<0 0>}
    }
   }
   slot {class=OverlaySlot, properties {horizontalAlignment=HAlign_Center, verticalAlignment=VAlign_Center}}
  }
  Image_DisconnectedIndicator {
   widgetClassPath=&Image
   bInherited=true
   bIsVariable=true
   parent=Overlay_1
   widget=Image_DisconnectedIndicator
   properties {
    visibility=Collapsed
    brush {
     bIsDynamicallyLoaded=false
     drawAs=Image
     imageType=NoImage
     mirroring=NoMirror
     resourceName=None
     resourceObject=/Game/Athena/HUD/Art/ConnectionError_B
     tiling=NoTile
     imageSize=<64 64>
     margin=<ltrb 0 0 0 0>
     outlineSettings {
      bUseBrushTransparency=false
      roundingType=HalfHeightRadius
      width=0
      color {colorUseRule=UseColor_Specified, specifiedColor=<rgba 0 0 0 0>}
      cornerRadii=<0 0 0 0>
     }
     tintColor {colorUseRule=UseColor_Specified, specifiedColor=<rgba 1 1 1 1>}
     uVRegion {bIsValid=false, max=<0 0>, min=<0 0>}
    }
   }
   slot {class=OverlaySlot, properties {horizontalAlignment=HAlign_Center, verticalAlignment=VAlign_Center}}
  }
  Image_ResurrectChip {
   widgetClassPath=&Image
   bInherited=true
   bIsVariable=true
   parent=Overlay_1
   widget=Image_ResurrectChip
   properties {
    visibility=Collapsed
    brush {
     bIsDynamicallyLoaded=false
     drawAs=Image
     imageType=FullColor
     mirroring=NoMirror
     resourceName=None
     resourceObject=/Game/Athena/HUD/Materials/MI_UI_RebootCardIcon_TeamInfo
     tiling=NoTile
     imageSize=<32 32>
     margin=<ltrb 0 0 0 0>
     outlineSettings {
      bUseBrushTransparency=false
      roundingType=HalfHeightRadius
      width=0
      color {colorUseRule=UseColor_Specified, specifiedColor=<rgba 0 0 0 0>}
      cornerRadii=<0 0 0 0>
     }
     tintColor {colorUseRule=UseColor_Specified, specifiedColor=<rgba 1 1 1 1>}
     uVRegion {bIsValid=false, max=<0 0>, min=<0 0>}
    }
    colorAndOpacity=<rgba 0.0135 0.8768 1 1>
   }
   slot {class=OverlaySlot, properties {horizontalAlignment=HAlign_Center, verticalAlignment=VAlign_Center}}
  }
  Image_ResurrectChip_Uncollected {
   widgetClassPath=&Image
   bIsVariable=true
   parent=Overlay_1
   widget=Image_ResurrectChip_Uncollected
   properties {
    visibility=Collapsed
    brush {
     bIsDynamicallyLoaded=false
     drawAs=Image
     imageType=NoImage
     mirroring=NoMirror
     resourceName=None
     resourceObject=/Game/Athena/HUD/Materials/M_UI_RebootIcon_Inst
     tiling=NoTile
     imageSize=<32 32>
     margin=<ltrb 0 0 0 0>
     outlineSettings {
      bUseBrushTransparency=false
      roundingType=HalfHeightRadius
      width=0
      color {colorUseRule=UseColor_Specified, specifiedColor=<rgba 0 0 0 0>}
      cornerRadii=<0 0 0 0>
     }
     tintColor {colorUseRule=UseColor_Specified, specifiedColor=<rgba 1 1 1 1>}
     uVRegion {bIsValid=false, max=<0 0>, min=<0 0>}
    }
    colorAndOpacity=<rgba 0.0135 0.8768 1 1>
   }
   slot {class=OverlaySlot, properties {horizontalAlignment=HAlign_Center, verticalAlignment=VAlign_Center}}
  }
  HorizontalBox_PlayerStateBarsExtensions {
   widgetClassPath=&HorizontalBox
   bIsVariable=true
   parent=HorizontalBox_HealthShields
   widget=HorizontalBox_PlayerStateBarsExtensions
   properties {renderTransformPivot=<0 1>}
   slot {class=HorizontalBoxSlot, properties {horizontalAlignment=HAlign_Left, padding=<ltrb 2 0 0 0>}}
  }
  WBP_SquadDirective_RebootCardTimer {
   widgetClassPath=&/WBP_BRGameUI_SquadDirective_RebootCardTimer
   bIsVariable=true
   parent=HorizontalBox_HealthShields
   widget=WBP_SquadDirective_RebootCardTimer
   properties {bOverride_Cursor=true, visibility=HitTestInvisible}
   slot {class=HorizontalBoxSlot, properties {verticalAlignment=VAlign_Center, padding=<ltrb 4 0 0 2>}}
  }
  HorizontalBox_Touch {
   widgetClassPath=&HorizontalBox
   bIsVariable=true
   parent=HorizontalBox_0
   widget=HorizontalBox_Touch
   slot {class=HorizontalBoxSlot}
  }
  GameModeIcon {
   widgetClassPath=&Image
   bIsVariable=true
   parent=HorizontalBox_0
   widget=GameModeIcon
   properties {
    visibility=HitTestInvisible
    brush {
     bIsDynamicallyLoaded=false
     drawAs=Image
     imageType=NoImage
     mirroring=NoMirror
     resourceName=None
     resourceObject=/Game/Athena/HUD/Art/M_GameMode_Icon
     tiling=NoTile
     imageSize=<40 40>
     margin=<ltrb 0 0 0 0>
     outlineSettings {
      bUseBrushTransparency=false
      roundingType=HalfHeightRadius
      width=0
      color {colorUseRule=UseColor_Specified, specifiedColor=<rgba 0 0 0 0>}
      cornerRadii=<0 0 0 0>
     }
     tintColor {colorUseRule=UseColor_Specified, specifiedColor=<rgba 1 1 1 1>}
     uVRegion {bIsValid=false, max=<0 0>, min=<0 0>}
    }
   }
   slot {class=HorizontalBoxSlot, properties {horizontalAlignment=HAlign_Left, verticalAlignment=VAlign_Center, padding=<ltrb 0 0 -40 0>}}
  }
  TeamMemberStateIcon {
   widgetClassPath=&Image
   bIsVariable=true
   parent=HorizontalBox_0
   widget=TeamMemberStateIcon
   properties {
    visibility=HitTestInvisible
    brush {
     bIsDynamicallyLoaded=false
     drawAs=Image
     imageType=NoImage
     mirroring=NoMirror
     resourceName=None
     resourceObject=&M_GameMode_Icon
     tiling=NoTile
     imageSize=<40 40>
     margin=<ltrb 0 0 0 0>
     outlineSettings {
      bUseBrushTransparency=false
      roundingType=HalfHeightRadius
      width=0
      color {colorUseRule=UseColor_Specified, specifiedColor=<rgba 0 0 0 0>}
      cornerRadii=<0 0 0 0>
     }
     tintColor {colorUseRule=UseColor_Specified, specifiedColor=<rgba 1 1 1 1>}
     uVRegion {bIsValid=false, max=<0 0>, min=<0 0>}
    }
   }
   slot {class=HorizontalBoxSlot, properties {horizontalAlignment=HAlign_Left, verticalAlignment=VAlign_Center, padding=<ltrb 0 0 -40 0>}}
  }
  ScaleBox_HboxMobile_RightofTeammate {
   widgetClassPath=&ScaleBox
   bIsVariable=true
   parent=HorizontalBox_0
   widget=ScaleBox_HboxMobile_RightofTeammate
   properties {stretch=UserSpecified}
   slot {class=HorizontalBoxSlot, properties {horizontalAlignment=HAlign_Center, verticalAlignment=VAlign_Center}}
  }
  HorizontalBox_Mobile_RightOfTeammateExtensions {
   widgetClassPath=&HorizontalBox
   bIsVariable=true
   parent=ScaleBox_HboxMobile_RightofTeammate
   widget=HorizontalBox_Mobile_RightOfTeammateExtensions
   properties {renderTransformPivot=<0 0.5>}
   slot {class=ScaleBoxSlot}
  }
  Overlay_PlayerStateExtensions {
   widgetClassPath=&Overlay
   bIsVariable=true
   parent=HorizontalBox_0
   widget=Overlay_PlayerStateExtensions
   properties {visibility=HitTestInvisible, renderTransformPivot=<0 0.5>}
   slot {class=HorizontalBoxSlot, properties {horizontalAlignment=HAlign_Left, verticalAlignment=VAlign_Center, padding=<ltrb 2 -12 0 -12>}}
  }
  WBP_BRGameUI_SquadDirective {
   widgetClassPath=&/WBP_BRGameUI_SquadDirective
   bIsVariable=true
   parent=Overlay_TeamMemberInfo
   widget=WBP_BRGameUI_SquadDirective
   properties {visibility=Collapsed}
   slot {class=OverlaySlot, properties {verticalAlignment=VAlign_Center, padding=<ltrb 48 0 0 0>}}
  }
  Overlay_ReviveBanner {
   widgetClassPath=&Overlay
   bIsVariable=true
   parent=Overlay_TeamMemberInfo
   widget=Overlay_ReviveBanner
   slot {class=OverlaySlot, properties {horizontalAlignment=HAlign_Fill, verticalAlignment=VAlign_Fill, padding=<ltrb 0 -10 -20 -10>}}
  }
  Image_DBNOBar {widgetClassPath=&Image, bInherited=true, widget=null}
  TutorialHighlight {widgetClassPath=&TutorialHighlight}
 ]
}
